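# Niagara effects for the Genius glider cosmetic: engine glows, ribbon trails, goo bursts and speed lines.
# Skeletal mesh sockets anchor flickering sprite lights, anim-trail ribbons, a curve-bent engine ribbon, and timed goo splashes with gravity-driven droplets.
# System Timeline values scale glow, trail and ribbon intensity; a smoothed vector bends the engine ribbon and RightDot swells side glows.
# GPU speed lines spawn by movement with front-end variants; two older speedline emitters are disabled.
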NiagaraSystem /BRCosmetics/Gadgets/Assets/VinderTech_GliderChute/Glider_Genius/FX/NS_Glider_Genius
dependencies {
 usedDataInterfaces [
  &NiagaraDataInterfaceArrayFloat
  &NiagaraDataInterfaceArrayFloat3
  &NiagaraDataInterfaceCurve
  &NiagaraDataInterfaceVector2DCurve
  &NiagaraDataInterfaceArrayInt32
 ]
 usedDynamicInputs [&MakeFloatFromVector2d, &ClampFloat, &Abs_Float]
 usedModules [&SystemState, &Timeline, &InterpolateOverTime]
}
emitters {
 Engine_Ribbon {
  emitterUpdateScript [
   &EmitterState()
   &SpawnBurst_Instantaneous(Spawn Count=10, Use Spawn Probability=false, UseLoopCountLimit=true)
  ]
  particleSpawnScript [
   &InitializeParticle(Color Mode=::NewEnumerator0, Lifetime=0, UsePositionOffset=false, Write Material Random=true)
   &InitializeRibbon(Color=<rgba 1 0.4 0 0.5>, Lifetime=0.2, Ribbon Width=1, UsePositionOffset=false, Write Material Random=true)
   /SetVariables(
    Particles.RibbonID=/Niagara/DynamicInputs/Meshes/MakeRibbonIDFromSpawnGroup()
    Particles.RibbonLinkOrder:NiagaraFloat=&ReturnNormalizedExecIndex()
   ) {isSetParameters=true}
  ]
  particleUpdateScript [
   &ParticleState(Kill Particles When Lifetime Has Elapsed=false, Let Infinitely Lived Particles Die When Emitter Deactivates=true)
   &SkeletalMeshLocation(
    MaskSpawnGroup=false
    Mesh Sampling Type=::NewEnumerator5
    Sampled Orientation Coordinate Space=::Local
    Sampled Position Offset=<0 2 0>
    Skeletal Mesh=$User.SkeletalMesh
    Socket Sampling Mode=::NewEnumerator5
   )
   /SetVariables(Particles.Position=/Niagara/DynamicInputs/Lerp/Lerp_Position(
     A:NiagaraPosition=$Particles.Position
     B:NiagaraPosition=&Add_Vector(
      A=Linked() {linkedVariable {name=Particles.Position, type=NiagaraPosition}}
      B=&MakeVector(
       X=&FloatFromCurve(
        FloatCurve=DataInterface() {propertyValues {Curve {keys [{tangentMode=RCTM_User, time=0, value=0}, {tangentMode=RCTM_User, time=1, value=1}]}}}
        CurveIndex=&ReturnNormalizedExecIndex()
        Scale Curve=&Lerp_Float(A=-30, B=-160, Alpha=$System.Timeline.NormalizedTime)
       )
       Y=&FloatFromCurve(
        FloatCurve=DataInterface() {propertyValues {Curve {keys [<at 0 0 RCIM_Cubic>, {arriveTangent=2.1213, interpMode=RCIM_Cubic, leaveTangent=2.1213, tangentMode=RCTM_User, time=1, value=1}]}}}
        CurveIndex=&ReturnNormalizedExecIndex()
        Scale Curve=&Multiply_Float(
         A=&MakeFloatFromVector(Vector=$System.InterpolateOverTime.InterpolateOverTime_V3_.Moving Average, Channel=::NewEnumerator1)
         B=&Lerp_Float(A=-10, B=-80, Alpha=$System.Timeline.NormalizedTime)
        )
       )
       Z=0
      )
     )
     Alpha:NiagaraFloat=&ReturnNormalizedExecIndex()
    )) {isSetParameters=true}
   &ScaleColor(
    Scale Alpha=&FloatFromCurve(
     FloatCurve=DataInterface() {propertyValues {Curve {keys [{interpMode=RCIM_Cubic, tangentMode=RCTM_User, time=0, value=1}, {interpMode=RCIM_Cubic, tangentMode=RCTM_User, time=1, value=0}]}}}
     CurveIndex=&ReturnNormalizedExecIndex()
     Scale Curve=&FloatFromCurve(
      FloatCurve=DataInterface() {propertyValues {Curve {keys [{interpMode=RCIM_Cubic, tangentMode=RCTM_User, time=0, value=0.5}, {interpMode=RCIM_Cubic, tangentMode=RCTM_User, time=0.5, value=1}]}}}
      CurveIndex=&Max_Float(A:NiagaraFloat=$System.Timeline.NormalizedTime, B:NiagaraFloat=$System.Timeline002.NormalizedTime)
      Scale Curve=0.1
     )
    )
    ScaleA=true
    ScaleRGB=true
   ) {module=ScaleColor001}
  ]
  rendererClasses [&NiagaraRibbonRendererProperties]
  renderers [{rendererClass=&NiagaraRibbonRendererProperties}]
 }
 Glow_Trails {
  emitterUpdateScript [
   &EmitterState()
   &FN_SpawnAnimTrail(
    Enable:NiagaraBool=&Boolean_AndOperation(
     Bool1=&Boolean_NotOperation(Bool=$User.bIsFrontEndPreview)
     Bool2=&SetBoolByFloatComparison(
      A=&Max_Float(A:NiagaraFloat=$System.Timeline.NormalizedTime, B:NiagaraFloat=$System.Timeline002.NormalizedTime)
      B=0
     )
    )
    Number of Anim Trails=5
   )
  ]
  particleSpawnScript [
   &FN_InitializeAnimTrailBoneIndecies(AnimTrail Selection=4, First Bone or Socket Index=12, Second Bone or Socket Index=13) {module=FN_InitializeAnimTrailBoneIndecies004}
   &FN_InitializeAnimTrailBoneIndecies(AnimTrail Selection=3, First Bone or Socket Index=10, Second Bone or Socket Index=11) {module=FN_InitializeAnimTrailBoneIndecies003}
   &FN_InitializeAnimTrailBoneIndecies(AnimTrail Selection=2, First Bone or Socket Index=8, Second Bone or Socket Index=9) {module=FN_InitializeAnimTrailBoneIndecies002}
   &FN_InitializeAnimTrailBoneIndecies(AnimTrail Selection=1, First Bone or Socket Index=6, Second Bone or Socket Index=7) {module=FN_InitializeAnimTrailBoneIndecies001}
   &FN_InitializeAnimTrail(
    AnimTrailWidth=&Multiply_Float(
     A=1
     B=&SelectFloatFromArray(Direct Array Index=$Particles.SpawnGroup, Float Selection Array=DataInterface() {propertyValues {FloatData [1,1,1,1]}})
    )
    Color=&MakeLinearColorFromVectorAndFloat("Vector (RGB)"=<1 0.4 0>, "Float (Alpha)"=0.15)
    Lifetime=0.133
   )
  ]
  particleUpdateScript [
   &ParticleState(Kill Particles When Lifetime Has Elapsed=false)
   &KillParticles(Kill Particles=&SetBoolByFloatComparison(A=$Particles.Age, B=0.15, Comparison Type=::NewEnumerator2))
   &InheritSourceMovement(
    Applied Force Scale=&VectorFromFloat(Value=-5)
    Applied Position Scale=<0.5 0.5 0.5>
    ApplyForce=false
    ApplyPosition=true
    ApplyPositionFalloff=false
   )
   &SolveForcesAndVelocity(Clamp Velocity=false, Limit Acceleration=false)
   &FN_UpdateAnimTrail(Kill Locked particle on system deactivate=false, Locked Particle RibbonWidth=1, SkeletalMesh=$User.SkeletalMesh) {module=FN_UpdateAnimTrail001}
   &ScaleColor(
    Scale Alpha=&FloatFromCurve(
     FloatCurve=DataInterface() {propertyValues {Curve {keys [{interpMode=RCIM_Cubic, tangentMode=RCTM_User, time=0, value=1}, {interpMode=RCIM_Cubic, tangentMode=RCTM_User, time=1, value=0}]}}}
     Scale Curve=&FloatFromCurve(
      FloatCurve=DataInterface() {propertyValues {Curve {keys [{interpMode=RCIM_Cubic, tangentMode=RCTM_User, time=0, value=0}, {interpMode=RCIM_Cubic, tangentMode=RCTM_User, time=0.5, value=1}]}}}
      CurveIndex=&Max_Float(A:NiagaraFloat=$System.Timeline.NormalizedTime, B:NiagaraFloat=$System.Timeline002.NormalizedTime)
      Scale Curve=0.25
     )
    )
    ScaleA=true
    ScaleRGB=true
   )
  ]
  rendererClasses [&NiagaraRibbonRendererProperties]
  renderers [{rendererClass=&NiagaraRibbonRendererProperties}]
 }
 Glows {
  emitterUpdateScript [
   &EmitterState()
   &SpawnBurst_Instantaneous(Spawn Count=2, Use Spawn Probability=false, UseLoopCountLimit=true)
   &SpawnBurst_Instantaneous(Spawn Count=2, Spawn Time=3.2, Use Spawn Probability=false, UseLoopCountLimit=true) {module=SpawnBurst_Instantaneous001}
  ]
  particleSpawnScript [
   &PartitionParticles(Number Of Partitions=4, Particles Per Partition=1, Partition Mode:ENiagara_PartitionMode=::NewEnumerator1)
   &InitializeParticle(
    Color=&MakeLinearColorFromVectorAndFloat(
     "Vector (RGB)"=<0.9 0.5 0>
     "Float (Alpha)"=&SelectFloatFromArray(
      Direct Array Index=$Particles.PartitionParticles.Partition
      Float Selection Array=DataInterface() {propertyValues {FloatData [0.3,0.25,0.2]}}
     )
    )
    Sprite Size Mode=::NewEnumerator3
    Uniform Sprite Size=&Multiply_Float(
     A=80
     B=&SelectFloatFromArray(Direct Array Index=$Particles.PartitionParticles.Partition, Float Selection Array=DataInterface() {propertyValues {FloatData [0.8,0.8,0.4]}})
    )
    UsePositionOffset=false
    Write Material Random=true
   )
  ]
  particleUpdateScript [
   &ParticleState(Kill Particles When Lifetime Has Elapsed=false, Loop Particles Lifetime=true)
   &ScaleColor(
    Scale Alpha=&Waveform(
     Global Input Position=&RandomRangeFloat(Minimum=0, Maximum=5)
     "Waveform [1]"=::NewEnumerator12
     "[1] Frequency"=12
     "[1] Amplitude Min/Max"=<0.6 1>
    )
    Scale RGB=&VectorFromFloat(Value=&Lerp_Float(A=1, B=2, Alpha=$System.Timeline.NormalizedTime))
    ScaleA=true
    ScaleRGB=true
   )
   &SkeletalMeshLocation(
    Bone / Socket Index=$Particles.PartitionParticles.Partition
    MaskSpawnGroup=false
    Mesh Sampling Type=::NewEnumerator5
    Sampled Position Offset=&SelectVectorFromArray(
     Direct Array Index=$Particles.PartitionParticles.Partition
     Vector Selection Array=DataInterface() {propertyValues {FloatData [<5 0 0>, <6 0 0>, <8 0 0>]}}
    )
    Skeletal Mesh=$User.SkeletalMesh
    Socket Sampling Mode=::NewEnumerator5
   )
   &CameraOffset(Camera Offset Amount=30)
   &ScaleSpriteSize(
    Uniform Curve Index=$Emitter.Age
    Uniform Curve Scale=&Lerp_Float(
     A=1
     B=1.25
     Alpha=&Max_Float(
      A:NiagaraFloat=$System.Timeline.NormalizedTime
      B:NiagaraFloat=&MakeCustomFloatFromBool(
       Bool=&SetBoolByIntComparison(A=$Particles.PartitionParticles.Partition, B=0, Comparison Type=::NewEnumerator4)
       True Float=0
       False Float=&MakeCustomFloatFromBool(
        Bool=&SetBoolByFloatComparison(
         A=&MakeFloatFromInt(Int=$Particles.PartitionParticles.Partition)
         B=3
         Comparison Type=::NewEnumerator6
        )
        True Float=&Multiply_Float(A=$User.RightDot, B=-1)
        False Float=$User.RightDot
       )
      )
     )
    )
    Uniform Curve Sprite Scale=DataInterface() {propertyValues {Curve {keys [
        {arriveTangent=9.0725, interpMode=RCIM_Cubic, leaveTangent=9.0725, tangentMode=RCTM_User, time=0, value=1}
        {arriveTangent=-2.5249, interpMode=RCIM_Cubic, leaveTangent=-2.5249, tangentMode=RCTM_User, time=0.1408, value=2}
        {arriveTangent=-0.0405, interpMode=RCIM_Cubic, leaveTangent=-0.0405, tangentMode=RCTM_User, time=0.5, value=1}
       ]}}}
   )
   &FN_Light_Attributes(
    Create Exponent Attribute=true
    Create Radius Attribute=true
    Enable Visibility Tag=true
    Exponent=2
    Radius=&Multiply_Float(
     A=/Niagara/DynamicInputs/Math/CalculateParticleRadius(
      Radius Calculation Type:ENiagaraCollisionRadiusOptions=::NewEnumerator0
      Particle Radius:NiagaraFloat=1
      Mesh Dimensions:Vector3f=<1 1 1>
      Method for Calculating Particle Radius:ENiagaraCalculateRadiusOptions=::NewEnumerator2
     )
     B=&SelectFloatFromArray(
      Direct Array Index=$Particles.PartitionParticles.Partition
      Float Selection Array=DataInterface() {propertyValues {FloatData [2.5,2.5,2.25,2.25]}}
     )
    )
   )
  ]
  rendererClasses [&NiagaraSpriteRendererProperties, &NiagaraLightRendererProperties]
  renderers [
   {rendererClass=&NiagaraSpriteRendererProperties}
   {rendererClass=&NiagaraLightRendererProperties, rendererIndex=1}
  ]
 }
 Goo_Burst {
  emitterUpdateScript [
   &EmitterState(
    Inactive Response=::NewEnumerator1
    Life Cycle Mode=::NewEnumerator1
    Loop Behavior=::NewEnumerator1
    Loop Delay=&MakeCustomFloatFromBool(Bool=$User.bIsFrontEndPreview, True Float=0, False Float=0)
    UseLoopDelay=true
   )
   &SpawnBurst_Instantaneous(Spawn Count=2, Use Spawn Probability=false, UseLoopCountLimit=false) {enabled=false}
   &SpawnBurst_Instantaneous(Spawn Group=1, Spawn Time=1.75, Use Spawn Probability=false, UseLoopCountLimit=false) {enabled=false, module=SpawnBurst_Instantaneous001}
   &SpawnBurst_Instantaneous(Spawn Group=2, Spawn Time=1.8, Use Spawn Probability=false, UseLoopCountLimit=false) {enabled=false, module=SpawnBurst_Instantaneous002}
   &SpawnBurst_Instantaneous(Spawn Count=3, Spawn Group=3, Spawn Time=3.25, Use Spawn Probability=false, UseLoopCountLimit=false) {module=SpawnBurst_Instantaneous003}
   &SpawnBurst_Instantaneous(Spawn Count=3, Spawn Group=4, Spawn Time=3.2, Use Spawn Probability=false, UseLoopCountLimit=false) {module=SpawnBurst_Instantaneous004}
   &SpawnBurst_Instantaneous(Spawn Group=5, Spawn Time=2.9, Use Spawn Probability=false, UseLoopCountLimit=false) {enabled=false, module=SpawnBurst_Instantaneous005}
   &SpawnBurst_Instantaneous(Spawn Group=6, Spawn Time=2.95, Use Spawn Probability=false, UseLoopCountLimit=false) {enabled=false, module=SpawnBurst_Instantaneous006}
  ]
  particleSpawnScript [
   &InitializeParticle(
    Lifetime Max=0.66
    Lifetime Min=0.33
    Lifetime Mode=::NewEnumerator1
    Position Offset=<0 0 75>
    Sprite Rotation Mode=::NewEnumerator2
    Sprite Size Mode=::NewEnumerator3
    Uniform Sprite Size=&Multiply_Float(
     A=&RandomRangeFloat(Minimum=60, Maximum=70)
     B=&SelectFloatFromArray(
      Direct Array Index=&SpawnGroup()
      Float Selection Array=DataInterface() {propertyValues {FloatData [1,1,1,1.2,1.2,0.5,0.5]}}
     )
    )
    UsePositionOffset=true
    Write Material Random=true
   ) {module=InitializeParticle001}
   &SkeletalMeshLocation(
    Bone / Socket Index=&SelectIntFromArray(
     Direct Array Index=&SpawnGroup()
     Int Selection Array=DataInterface() {propertyValues {IntData [14,11,13,2,3,15,16]}}
    )
    MaskSpawnGroup=false
    Mesh Sampling Type=::NewEnumerator5
    Skeletal Mesh=$User.SkeletalMesh
    Socket Sampling Mode=::NewEnumerator5
   )
   /Niagara/Modules/Update/SubUV/SubUVAnimation(SubUV Animation Mode=::NewEnumerator1) {enabled=false}
   &AddVelocity(
    Rotation Coordinate Space=::World
    Velocity=&RotateVectorByQuaternion(Vector To Rotate=<1 0 0>, Quaternion=$Particles.SkeletalMeshLocation.SampledRotation)
    Velocity Speed Scale=&Multiply_Float(
     A=&RandomRangeFloat(Minimum=30, Maximum=60)
     B=&SelectFloatFromArray(
      Direct Array Index=&SpawnGroup()
      Float Selection Array=DataInterface() {propertyValues {FloatData [8,-2,-2,4,4,1,1]}}
     )
    )
   ) {module=AddVelocity001}
  ]
  particleUpdateScript [
   &ParticleState()
   &InheritSourceMovement(
    Applied Position Scale=&VectorFromFloat(Value=0.9)
    ApplyForce=false
    ApplyPosition=true
    ApplyPositionFalloff=false
   )
   &Drag(Drag=&RandomRangeFloat(Minimum=0.1, Maximum=0.3), Use Linear Drag=true, Use Rotational Drag=false)
   &GravityForce()
   &SolveForcesAndVelocity(Clamp Velocity=false, Limit Acceleration=false)
   &ScaleSpriteSize(Scale Factor=&Vector2DFromFloat(Value=&FloatFromCurve(FloatCurve=DataInterface() {propertyValues {Curve {keys [
          {arriveTangent=1.6115, interpMode=RCIM_Cubic, leaveTangent=1.6115, tangentMode=RCTM_User, time=0, value=0.2}
          {interpMode=RCIM_Cubic, tangentMode=RCTM_User, time=1, value=1}
         ]}}})))
   &ScaleColor(
    Scale Alpha=&FloatFromCurve(FloatCurve=DataInterface() {propertyValues {Curve {keys [{interpMode=RCIM_Cubic, tangentMode=RCTM_User, time=0, value=1}, {interpMode=RCIM_Cubic, tangentMode=RCTM_User, time=1, value=0}]}}})
    ScaleA=true
    ScaleRGB=true
   )
   &DynamicMaterialParameters(
    Index 0 Param 1=&RandomRangeFloat(Minimum=0, Maximum=1)
    Index 0 Param 3=&FloatFromCurve(FloatCurve=DataInterface() {propertyValues {Curve {keys [{interpMode=RCIM_Cubic, tangentMode=RCTM_User, time=0, value=0}, {interpMode=RCIM_Cubic, tangentMode=RCTM_User, time=1, value=1}]}}})
    Index 0 Param 4=&FloatFromCurve(FloatCurve=DataInterface() {propertyValues {Curve {keys [<at 0 0 RCIM_Cubic>, {arriveTangent=1.5276, interpMode=RCIM_Cubic, leaveTangent=1.5276, tangentMode=RCTM_User, time=1, value=1.5}]}}})
    Module.Param0WriteEnabled=true
    Module.Param1WriteEnabled=false
    Module.Param2WriteEnabled=true
    Module.Param3WriteEnabled=true
   )
   &CameraOffset(Camera Offset Amount=15)
  ]
  rendererClasses [&NiagaraSpriteRendererProperties]
  renderers [{rendererClass=&NiagaraSpriteRendererProperties}]
 }
 Goo_Droplets {
  emitterUpdateScript [
   &EmitterState(
    Inactive Response=::NewEnumerator1
    Life Cycle Mode=::NewEnumerator1
    Loop Behavior=::NewEnumerator1
    Loop Delay=&MakeCustomFloatFromBool(Bool=$User.bIsFrontEndPreview, True Float=0, False Float=0)
    UseLoopDelay=true
   )
   &SpawnBurst_Instantaneous(Spawn Count=9, Use Spawn Probability=false, UseLoopCountLimit=false) {enabled=false}
   &SpawnBurst_Instantaneous(Spawn Count=5, Spawn Group=1, Spawn Time=1.75, Use Spawn Probability=false, UseLoopCountLimit=false) {enabled=false, module=SpawnBurst_Instantaneous001}
   &SpawnBurst_Instantaneous(Spawn Count=5, Spawn Group=2, Spawn Time=1.8, Use Spawn Probability=false, UseLoopCountLimit=false) {enabled=false, module=SpawnBurst_Instantaneous002}
   &SpawnBurst_Instantaneous(Spawn Count=7, Spawn Group=3, Spawn Time=3.25, Use Spawn Probability=false, UseLoopCountLimit=false) {module=SpawnBurst_Instantaneous003}
   &SpawnBurst_Instantaneous(Spawn Count=7, Spawn Group=4, Spawn Time=3.2, Use Spawn Probability=false, UseLoopCountLimit=false) {module=SpawnBurst_Instantaneous004}
   &SpawnBurst_Instantaneous(Spawn Count=3, Spawn Group=5, Spawn Time=2.9, Use Spawn Probability=false, UseLoopCountLimit=false) {enabled=false, module=SpawnBurst_Instantaneous005}
   &SpawnBurst_Instantaneous(Spawn Count=3, Spawn Group=6, Spawn Time=2.95, Use Spawn Probability=false, UseLoopCountLimit=false) {enabled=false, module=SpawnBurst_Instantaneous006}
  ]
  particleSpawnScript [
   &InitializeParticle(
    Color=<rgba 0.5457 0.1976 0.0546 1>
    Lifetime Max=1
    Lifetime Min=0.33
    Lifetime Mode=::NewEnumerator1
    Position Offset=<0 0 75>
    Sprite Rotation Mode=::NewEnumerator2
    Sprite Size=&Vector2DFromCurve(
     Vector2Curve=DataInterface() {propertyValues {
       XCurve {keys [<at -0.0011 2.5 RCIM_Cubic>,<at 0.9988 1 RCIM_Cubic>]}
       YCurve {keys [
         {arriveTangent=-30.8979, interpMode=RCIM_Cubic, leaveTangent=-30.8979, tangentMode=RCTM_User, time=0, value=9}
         {arriveTangent=-6.5075, interpMode=RCIM_Cubic, leaveTangent=-6.5075, time=0.2382, value=5}
         <at 0.9988 2.5 RCIM_Cubic>
        ]}
      }}
     CurveIndex=&RandomRangeFloat(Minimum=0, Maximum=1)
     ScaleCurve=<2 2>
    )
    Sprite Size Mode=::NewEnumerator4
    UsePositionOffset=true
    Write Material Random=true
   ) {module=InitializeParticle001}
   &SkeletalMeshLocation(
    Bone / Socket Index=&SelectIntFromArray(
     Direct Array Index=&SpawnGroup()
     Int Selection Array=DataInterface() {propertyValues {IntData [14,11,13,2,3,15,16]}}
    )
    MaskSpawnGroup=false
    Mesh Sampling Type=::NewEnumerator5
    Skeletal Mesh=$User.SkeletalMesh
    Socket Sampling Mode=::NewEnumerator5
   )
   &AddVelocity(
    Cone Axis=&RotateVectorByQuaternion(Vector To Rotate=<1 0 0>, Quaternion=$Particles.SkeletalMeshLocation.SampledRotation)
    Module.UseConeFalloff=false
    Module.UseVelDistribution=false
    Override Seed=false
    Rotation Coordinate Space=::World
    Velocity Mode=::NewEnumerator2
    Velocity Speed=&Multiply_Float(
     A=&RandomRangeFloat(Minimum=50, Maximum=200)
     B=&SelectFloatFromArray(
      Direct Array Index=&SpawnGroup()
      Float Selection Array=DataInterface() {propertyValues {FloatData [3,-2,-2,4,4,1,1]}}
     )
    )
   ) {module=AddVelocity001}
   &DynamicMaterialParameters(
    Index 0 Param 1=&UniformRangedFloat(Minimum=10, Maximum=10)
    Module.Param0WriteEnabled=true
    Module.Param1WriteEnabled=false
    Module.Param2WriteEnabled=false
    Module.Param3WriteEnabled=false
   )
  ]
  particleUpdateScript [
   &ParticleState()
   &InheritSourceMovement(
    Applied Position Scale=&VectorFromFloat(Value=0.9)
    ApplyForce=false
    ApplyPosition=true
    ApplyPositionFalloff=false
   )
   &Drag(Drag=&RandomRangeFloat(Minimum=0.1, Maximum=0.3), Use Linear Drag=true, Use Rotational Drag=false)
   &GravityForce()
   &SolveForcesAndVelocity(Clamp Velocity=false, Limit Acceleration=false)
   &ScaleColor(
    Scale Alpha=&FloatFromCurve(FloatCurve=DataInterface() {propertyValues {Curve {keys [{interpMode=RCIM_Cubic, tangentMode=RCTM_User, time=0, value=1}, {interpMode=RCIM_Cubic, tangentMode=RCTM_User, time=1, value=0}]}}})
    ScaleA=true
    ScaleRGB=true
   )
   &ScaleSpriteSizeBySpeed(Max Scale Factor=<0.75 3>)
  ]
  rendererClasses [&NiagaraSpriteRendererProperties]
  renderers [{rendererClass=&NiagaraSpriteRendererProperties}]
 }
 SpeedLine {
  simTarget=GPUComputeSim
  emitterUpdateScript [
   &EmitterState(
    Delay First Loop Only=true
    Inactive Response=::NewEnumerator1
    Life Cycle Mode=::NewEnumerator1
    Loop Delay=1.5
    Loop Duration=3
    UseLoopDelay=true
   )
   &SpawnRate(SpawnRate=200, Use Spawn Probability=false) {enabled=false}
   &SpawnPerUnit(
    Module.Use Max Movement Threshold=true
    Module.Use Movement Tolerance=false
    Movement Tolerance=50
    Spawn Spacing:NiagaraFloat=&MakeCustomFloatFromBool(Bool=&SetBoolByFloatComparison(A=$System.Timeline.NormalizedTime, B=0), True Float=30, False Float=0)
    Use Spawn Probability=false
   )
  ]
  particleSpawnScript [
   &InitializeParticle(
    Color=<rgba 1 1 1 0.25>
    Lifetime Max=0.12
    Lifetime Min=0.1
    Lifetime Mode=::NewEnumerator1
    Position Offset=<0 -100 0>
    Sprite Size Max=<2 7>
    Sprite Size Min=<1.5 2>
    Sprite Size Mode=::NewEnumerator2
    UsePositionOffset=true
    Write Material Random=true
   )
   &TorusLocation(
    Align Torus With Axis=true
    Handle Radius=0
    Large Radius:NiagaraFloat=&RandomRangeFloat(Minimum=&MakeCustomFloatFromBool(Bool=$User.IsFrontEnd, True Float=150, False Float=250), Maximum=300)
    MaskSpawnGroup=false
    Module.Override Randomness=false
    Offset=&FN_FrontEndVector3Modifier(Gameplay=<0 100 0>, Front End=<0 -100 0>)
    Torus Axis=<1 0 0>
   )
   &AddVelocity(Velocity=&RandomRangeVector(Minimum=<-500 0 0>, Maximum=<-700 0 0>))
   &AddVelocityFromPoint(Module.Use Falloff Distance=false, Origin Offset=<200 0 0>, Velocity Strength=1500)
   &CameraOffset(Camera Offset Amount=-10) {enabled=false}
  ]
  particleUpdateScript [
   &SolveForcesAndVelocity(Clamp Velocity=false, Limit Acceleration=false)
   &ParticleState()
   &ScaleSpriteSizeBySpeed(Max Scale Factor=<1 25>, Min Scale Factor=<0.01 0.05>, Velocity Threshold=1)
   &ScaleColor(
    Scale Alpha=&Multiply_Float(
     A=&FloatFromCurve(FloatCurve=DataInterface() {propertyValues {Curve {keys [
          {interpMode=RCIM_Cubic, tangentMode=RCTM_User, time=0, value=0}
          {interpMode=RCIM_Cubic, tangentMode=RCTM_User, time=0.5, value=1}
          {tangentMode=RCTM_User, time=1, value=0}
         ]}}})
     B=&Multiply_Float(
      A=&MakeCustomFloatFromBool(
       Bool=$User.IsFrontEnd
       True Float=&RandomRangeFloat(Minimum=0.1, Maximum=0.2)
       False Float=&RandomRangeFloat(Minimum=0.4, Maximum=0.6)
      )
      B=1
     )
    )
    ScaleA=true
    ScaleRGB=true
   )
  ]
  rendererClasses [&NiagaraSpriteRendererProperties]
  renderers [{rendererClass=&NiagaraSpriteRendererProperties}]
 }
 Speedlines {
  bEnabled=false
  simTarget=GPUComputeSim
  emitterUpdateScript [
   &EmitterState()
   &SpawnRate(SpawnRate=&FN_FrontEndFloatModifier(Gameplay=0, Front End=50), Use Spawn Probability=false)
   &SpawnPerUnit(
    Module.Use Max Movement Threshold=true
    Module.Use Movement Tolerance=false
    Movement Tolerance=50
    Spawn Spacing:NiagaraFloat=&MakeCustomFloatFromBool(Bool=&SetBoolByFloatComparison(A=$System.Timeline.NormalizedTime, B=0), True Float=30, False Float=0)
    Use Spawn Probability=false
   )
  ]
  particleSpawnScript [
   &InitializeParticle(
    Color=&MakeLinearColorFromVectorAndFloat(
     "Vector (RGB)"=&MakeVectorFromLinearColorRGB(Color=<rgba 1 1 1 1>)
     "Float (Alpha)"=&RandomRangeFloat(Minimum=0.05, Maximum=0.15)
    )
    Lifetime Max=0.4
    Lifetime Min=0.2
    Lifetime Mode=::NewEnumerator1
    Position Offset=&FN_FrontEndVector3Modifier(Gameplay=<0 0 150>, Front End=<0 0 0>)
    Sprite Size Max=<6 6>
    Sprite Size Min=<3 3>
    Sprite Size Mode=::NewEnumerator2
    UsePositionOffset=true
    Write Material Random=true
   )
   &ShapeLocation(
    Handle Radius=100
    Large Radius=300
    MaskSpawnGroup=false
    Offset=<200 0 0>
    Offset Coordinate Space=::Local
    Offset Mode=::NewEnumerator0
    Override Seed=false
    Rotation Mode=::NewEnumerator1
    Shape Primitive=::NewEnumerator3
    UseOwnerScale=true
   )
   &TorusLocation(
    Align Torus With Axis=true
    Coordinate Space:ENiagaraCoordinateSpace=::Simulation
    Handle Radius=75
    Large Radius=200
    MaskSpawnGroup=false
    Module.Override Randomness=false
    Torus Axis=&TransformVector(Vector=<0 1 0>, Destination Space=::Simulation)
   ) {enabled=false}
   &InheritVelocity(Module.Limit Speed=true, Velocity Scale=<0.7 0.7 0>) {enabled=false}
   &AddVelocity(Velocity=<-1 0 0>, Velocity Speed Scale=&FN_FrontEndFloatModifier(Gameplay=0, Front End=1200))
  ]
  particleUpdateScript [
   &ParticleState()
   &InheritSourceMovement(Applied Position Scale=<0.25 0.25 0.25>, ApplyForce=false, ApplyPosition=true, ApplyPositionFalloff=false)
   &SolveForcesAndVelocity(Clamp Velocity=false, Limit Acceleration=false)
   &SpriteFacingAndAlignment(
    Alignment Coordinate Space=::Local
    Sprite Alignment=$System.InterpolateOverTime.InterpolateOverTime_V3_.Moving Average
    WriteAlignment=true
    WriteFacing=false
   ) {module=SpriteFacingAndAlignment001}
   &SpriteFacingAndAlignment(
    Alignment Coordinate Space=::Local
    Sprite Alignment=$System.InterpolateOverTime.InterpolateOverTime_V3_.Moving Average
    WriteAlignment=true
    WriteFacing=false
   ) {enabled=false}
   &ScaleSpriteSize(Scale Factor=&Lerp_Vector2(
     A=<1 1>
     B=&RandomRangeVector2D(Minimum=<0.8 8>, Maximum=<0.75 20>)
     Alpha=&FN_FrontEndFloatModifier(
      Gameplay=&Max_Float(A:NiagaraFloat=$System.Timeline.NormalizedTime, B:NiagaraFloat=$System.Timeline002.NormalizedTime)
      Front End=1
     )
    )) {module=ScaleSpriteSize001}
   &ScaleSpriteSize(Scale Factor=&Lerp_Vector2(
     A=<1 1>
     B=&RandomRangeVector2D(Minimum=<0.8 12>, Maximum=<0.75 20>)
     Alpha=$System.Timeline.NormalizedTime
    )) {enabled=false}
   &ScaleColor(
    Scale Alpha=&FloatFromCurve(
     FloatCurve=DataInterface() {propertyValues {Curve {keys [
         <at 0 0 RCIM_Cubic>
         {interpMode=RCIM_Cubic, tangentMode=RCTM_User, time=0.2, value=1}
         {interpMode=RCIM_Cubic, tangentMode=RCTM_User, time=0.8, value=1}
         <at 1 0 RCIM_Cubic>
        ]}}}
     Scale Curve=&FN_FrontEndFloatModifier(
      Gameplay=&FloatFromCurve(
       FloatCurve=DataInterface() {propertyValues {Curve {keys [{interpMode=RCIM_Cubic, tangentMode=RCTM_User, time=0, value=0}, {interpMode=RCIM_Cubic, tangentMode=RCTM_User, time=0.5, value=1}]}}}
       CurveIndex=$System.Timeline.NormalizedTime
      )
      Front End=1
     )
    )
    ScaleA=true
    ScaleRGB=true
   )
  ]
  rendererClasses [&NiagaraSpriteRendererProperties]
  renderers [{rendererClass=&NiagaraSpriteRendererProperties}]
 }
 Speedlines001 {
  bEnabled=false
  simTarget=GPUComputeSim
  emitterUpdateScript [
   &EmitterState(Inactive Response=::NewEnumerator1, Life Cycle Mode=::NewEnumerator1, Loop Duration=1, UseLoopDelay=false)
   &SpawnBurst_Instantaneous(Spawn Count=500, Use Spawn Probability=false, UseLoopCountLimit=true)
   &SpawnRate(
    SpawnRate=&MakeCustomFloatFromBool(Bool=$User.bIsFrontEndPreview, True Float=50, False Float=0)
    Use Spawn Probability=false
   ) {enabled=false}
   &SpawnPerUnit(
    Module.Use Max Movement Threshold=true
    Module.Use Movement Tolerance=true
    Movement Tolerance=10
    Spawn Spacing:NiagaraFloat=&MakeCustomFloatFromBool(Bool=&SetBoolByFloatComparison(A=$System.Timeline.NormalizedTime, B=0), True Float=30, False Float=0)
    Use Spawn Probability=false
   ) {enabled=false}
   /SetVariables(Emitter.VolumeRadius=500) {isSetParameters=true}
  ]
  particleSpawnScript [
   &InitializeParticle(
    Color=&MakeLinearColorFromVectorAndFloat(
     "Vector (RGB)"=&MakeVectorFromLinearColorRGB(Color=<rgba 1 1 1 1>)
     "Float (Alpha)"=&RandomRangeFloat(Minimum=1, Maximum=1)
    )
    Lifetime Max=0.4
    Lifetime Min=0.2
    Lifetime Mode=::NewEnumerator1
    Position Offset=<0 0 150>
    Sprite Size Max=<6 6>
    Sprite Size Min=<4 4>
    Sprite Size Mode=::NewEnumerator2
    UsePositionOffset=true
    Write Material Random=true
   )
   /SetVariables(Particles.Position=&AddVectorToPosition(Position=$Engine.Emitter.SimulationPosition, Vector=<0 0 150>)) {enabled=false, isSetParameters=true}
   &ShapeLocation(
    MaskSpawnGroup=false
    Offset=&Multiply_VectorByFloat(
     Vector=$System.InterpolateOverTime.InterpolateOverTime_V3_.Moving Average
     Float=&Lerp_Float(A=0, B=0, Alpha=$System.Timeline.NormalizedTime)
    )
    Offset Coordinate Space=::Local
    Offset Mode=::NewEnumerator0
    Override Seed=false
    Rotation Mode=::NewEnumerator1
    Sphere Radius=$Emitter.VolumeRadius
    UseOwnerScale=true
   )
   &TorusLocation(
    Align Torus With Axis=true
    Coordinate Space:ENiagaraCoordinateSpace=::Simulation
    Handle Radius=75
    Large Radius=200
    MaskSpawnGroup=false
    Module.Override Randomness=false
    Torus Axis=&TransformVector(Vector=<0 1 0>, Destination Space=::Simulation)
   ) {enabled=false}
   &InheritVelocity(Module.Limit Speed=true, Velocity Scale=<0.7 0.7 0>) {enabled=false}
   &AddVelocity(Velocity=<-1 0 0>, Velocity Speed Scale=10) {enabled=false}
  ]
  particleUpdateScript [
   &ParticleState(Kill Particles When Lifetime Has Elapsed=false)
   &InheritSourceMovement(
    Applied Force Scale=&VectorFromFloat(Value=-1)
    Applied Position Scale=&VectorFromFloat(Value=0.5)
    ApplyForce=false
    ApplyPosition=true
    ApplyPositionFalloff=false
   )
   &SolveForcesAndVelocity(Clamp Velocity=false, Limit Acceleration=false)
   /SetVariables(Particles.PositionCenter:NiagaraPosition=&AddVectorToPosition(Position=$Engine.Owner.Position, Vector=<0 0 150>)) {isSetParameters=true}
   &RegionMask(
    Offset:Vector3f=<0 0 150>
    Region Falloff=0
    Region Origin:NiagaraPosition=$Particles.PositionCenter
    Region Radius:NiagaraFloat=$Emitter.VolumeRadius
    UseOffset=false
    UseRotation=false
   )
   /SetVariables(Particles.Position=&MakeCustomVectorFromBool(
     Bool=&SetBoolByFloatComparison(A=$Particles.RegionMask.Float, B=1, Comparison Type=::NewEnumerator4)
     True Vector=Linked() {linkedVariable {name=Particles.Position, type=NiagaraPosition}}
     False Vector=&AddVectorToPosition(
      Position=$Particles.Position
      Vector=&Multiply_VectorByFloat(
       Vector=&NormalizeVector(Vector To Normalize=$Engine.Owner.Velocity, Safe Normalize=true, Safe Fallback Threshold=0, Safe Fallback Vector=<0 0 0.001>)
       Float=&Multiply_Float(
        A=&DotProduct(
         Vector A=&SubtractVectorFromPosition(Position=$Particles.Position, Vector=Linked() {linkedVariable {name=Particles.PositionCenter, type=NiagaraPosition}})
         Vector B=/Niagara/DynamicInputs/Vectors/NegateVector(Vector:Vector3f=&NormalizeVector(Vector To Normalize=$Engine.Owner.Velocity, Safe Normalize=false, Safe Fallback Threshold=0, Safe Fallback Vector=<0 0 0.001>))
        )
        B=&Add_Float(A=1.9, B=0)
       )
      )
     )
    )) {isSetParameters=true}
   /SetVariables(Particles.Position=&MakeCustomVectorFromBool(
     Bool=&SetBoolByFloatComparison(A=$Particles.RegionMask.Float, B=1, Comparison Type=::NewEnumerator4)
     True Vector=Linked() {linkedVariable {name=Particles.Previous.Position, type=NiagaraPosition}}
     False Vector=&AddVectorToPosition(
      Position=$Particles.Previous.Position
      Vector=&Multiply_VectorByFloat(
       Vector=&SubtractVectorFromPosition(Position=$Particles.PositionCenter, Vector=Linked() {linkedVariable {name=Particles.Position, type=NiagaraPosition}})
       Float=1.99
      )
     )
    )) {enabled=false, isSetParameters=true}
   &SkeletalMeshLocation(
    Bone Sampling Mode=::NewEnumerator4
    MaskSpawnGroup=false
    Sampled Orientation Coordinate Space=::Local
    Sampled Position Offset=$Particles.Initial.ShapeLocation.ShapePosition
    Skeletal Mesh=$User.SkeletalMesh
   ) {enabled=false}
   &SpriteFacingAndAlignment(
    Alignment Coordinate Space=::World
    Sprite Alignment=&NormalizeVector(Vector To Normalize=$Engine.Owner.Velocity, Safe Normalize=true, Safe Fallback Threshold=0, Safe Fallback Vector=<0 0 0.001>)
    WriteAlignment=true
    WriteFacing=false
   )
   &ScaleSpriteSize(Scale Factor=&Lerp_Vector2(
     A=<1 1>
     B=&RandomRangeVector2D(Minimum=<0.8 12>, Maximum=<0.75 20>)
     Alpha=&Max_Float(A:NiagaraFloat=$System.Timeline.NormalizedTime, B:NiagaraFloat=$System.Timeline002.NormalizedTime)
    ))
   &ScaleColor(
    Scale Alpha=&FloatFromCurve(
     FloatCurve=DataInterface() {propertyValues {Curve {keys [<at 0 0>,<at 0.2 1>,<at 0.8 1>,<at 1 0>]}}}
     Scale Curve=&FloatFromCurve(
      FloatCurve=DataInterface() {propertyValues {Curve {keys [{interpMode=RCIM_Cubic, tangentMode=RCTM_User, time=0, value=1}, {interpMode=RCIM_Cubic, tangentMode=RCTM_User, time=0.25, value=1}]}}}
      CurveIndex=$System.Timeline.NormalizedTime
     )
    )
    Scale RGB=&Lerp_Vector(A=<2 0 0>, B=<1 1 1>, Alpha=$Particles.RegionMask.Float)
    ScaleA=false
    ScaleRGB=true
   )
  ]
  rendererClasses [&NiagaraSpriteRendererProperties]
  renderers [
   {rendererClass=&NiagaraSpriteRendererProperties}
   {rendererClass=&NiagaraSpriteRendererProperties, rendererIndex=1}
  ]
 }
}
summary {
 systemName=NS_Glider_Genius
 emitters [
  {emitterName=Glows, rendererClasses [&NiagaraSpriteRendererProperties, &NiagaraLightRendererProperties]}
  {emitterName=Glow_Trails, rendererClasses [&NiagaraRibbonRendererProperties]}
  {emitterName=Engine_Ribbon, rendererClasses [&NiagaraRibbonRendererProperties]}
  {bEnabled=false, emitterName=Speedlines, simTarget=GPUComputeSim, rendererClasses [&NiagaraSpriteRendererProperties]}
  {bEnabled=false, emitterName=Speedlines001, simTarget=GPUComputeSim, rendererClasses [&NiagaraSpriteRendererProperties]}
  {emitterName=Goo_Droplets, rendererClasses [&NiagaraSpriteRendererProperties]}
  {emitterName=Goo_Burst, rendererClasses [&NiagaraSpriteRendererProperties]}
  {emitterName=SpeedLine, simTarget=GPUComputeSim, rendererClasses [&NiagaraSpriteRendererProperties]}
 ]
 userVariables [
  User.bIsFrontEnd {type=NiagaraBool}
  User.bIsFrontEndPreview {type=NiagaraBool}
  User.bIsFullyDeployed {type=NiagaraBool}
  User.ForwardDot {type=NiagaraFloat, defaultValue=0, description=""}
  User.IsFrontEnd {type=NiagaraBool}
  User.RibbonIDOffset {type=NiagaraInt32}
  User.RightDot {type=NiagaraFloat, defaultValue=0, description=""}
  User.SkeletalMesh {
   type=NiagaraDataInterfaceSkeletalMesh
   defaultValue {
    struct=NiagaraExt_VariableValue_DataInterface
    value {dataInterface=&self:NiagaraDataInterfaceSkeletalMesh_0}
   }
  }
 ]
}
